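# Niagara system for the Jolly glider's trails: skeletal-mesh ribbons, beard sparkles and speed lines.
# Five anim-trail emitters each trail two bone/socket pairs, tinted gold, red or green and faded by system timeline and integrator values.
# Hue-shifted sparkles spawn by movement in gameplay, with a separate rate-based emitter for front end.
# Front-end flags reverse trail and speed-line motion; ForwardDot drives gameplay speed-line alpha and sparkle offset.
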NiagaraSystem /BRCosmetics/Gadgets/Assets/VinderTech_GliderChute/Glider_Default_Jolly/FX/NS_Glider_Default_Jolly_Trails
dependencies {
 usedDataInterfaces [
  &NiagaraDataInterfaceVector2DCurve
  &NiagaraDataInterfaceCurve
  &NiagaraDataInterfaceColorCurve
  &NiagaraDataInterfaceArrayFloat
 ]
 usedDynamicInputs [
  &Add_Float
  &SetBoolByFloatComparison
  &ClampFloat
  &Abs_Float
 ]
 usedModules [
  &self:SystemSpawnScript.SystemScriptSource.SystemScriptGraph.NiagaraNodeAssignment_0.SetVariables_#1
  &SystemState
  &FN_Integrator
  &Timeline
  &self:SystemSpawnScript.SystemScriptSource.SystemScriptGraph.NiagaraNodeAssignment_1.SetVariables_#2
  &self:SystemSpawnScript.SystemScriptSource.SystemScriptGraph.NiagaraNodeAssignment_2.SetVariables_#3
 ]
}
emitters {
 FN_AnimTrail {
  emitterSpawnScript [&FN_RibbonIDOffsetUtility()]
  emitterUpdateScript [
   &EmitterState()
   &FN_SpawnAnimTrail(Number of Anim Trails=2)
  ]
  particleSpawnScript [
   &FN_InitializeAnimTrailBoneIndecies(First Bone or Socket Index=1, Second Bone or Socket Index=0)
   &FN_InitializeAnimTrailBoneIndecies(AnimTrail Selection=1, First Bone or Socket Index=3, Second Bone or Socket Index=2) {module=FN_InitializeAnimTrailBoneIndecies001}
   /SetVariables(Particles.AlphaFadeIn:NiagaraFloat=$System.SystemFadeIn) {isSetParameters=true}
   &FN_InitializeAnimTrail(
    AnimTrailWidth=1.5
    Color=&MakeLinearColor(
     R:NiagaraFloat=1
     G:NiagaraFloat=1
     B:NiagaraFloat=1
     A:NiagaraFloat=&Multiply_Float(A=$Particles.AlphaFadeIn, B=0.5)
    )
    Lifetime=&MakeCustomFloatFromBool(Bool=$Particles.Locked, True Float=100, False Float=1.5)
    Position Offset=&FN_FrontEndVector3Modifier(Gameplay=<0 -20 0>, Front End=<0 20 0>)
   )
   &AddVelocity(Velocity=&Multiply_VectorByFloat(Vector=<0 400 0>, Float=&MakeCustomFloatFromBool(Bool=$User.bIsFrontEnd, True Float=-1, False Float=1)))
  ]
  particleUpdateScript [
   &SolveForcesAndVelocity(Clamp Velocity=false, Limit Acceleration=false)
   &FN_UpdateAnimTrail(
    Locked Particle Color:LinearColor=&MultiplyLinearColor(
     Linear Color=$Particles.Initial.Color
     Scale Factor=&MakeFloat4FromVectorAndFloat("Vector (RGB)"=<1 1 1>, "Float (Alpha)"=$System.FN_Integrator001.Value)
    )
    Locked Particle RibbonWidth=1
    SkeletalMesh=$User.SkeletalMesh
   ) {module=FN_UpdateAnimTrail001}
   /SetVariables(Particles.RibbonU0Override=$Particles.NormalizedAge) {isSetParameters=true, module=SetVariables_#4}
   &Color(
    Color=<rgba 0.6376 0.3387 0.0398 0.1>
    Scale Alpha=&Multiply_Float(
     A=&Multiply_Float(A=0.5, B=$Particles.AlphaFadeIn)
     B=&FN_FrontEndFloatModifier(Gameplay=$System.FN_Integrator001.Value, Front End=1)
    )
   ) {enabled=false}
   &ScaleColor(
    Scale Alpha=&Lerp_Float(A=0, B=0.1, Alpha=$System.Timeline.Time)
    Scale RGB=&Lerp_Vector(
     A=&MakeVectorFromLinearColorRGB(Color=<rgba 0.9 0.4781 0.0562 1>)
     B=&MakeVectorFromLinearColorRGB(Color=<rgba 0.6376 0.0398 0.0398 1>)
     Alpha=$System.Timeline.Time
    )
    ScaleA=true
    ScaleRGB=true
   )
  ]
  rendererClasses [&NiagaraRibbonRendererProperties]
  renderers [{rendererClass=&NiagaraRibbonRendererProperties}]
 }
 FN_AnimTrail001 {
  emitterSpawnScript [&FN_RibbonIDOffsetUtility()]
  emitterUpdateScript [
   &EmitterState()
   &FN_SpawnAnimTrail(AnimTrailWidthMode=::FromFirst, Number of Anim Trails=2)
  ]
  particleSpawnScript [
   &FN_InitializeAnimTrailBoneIndecies(First Bone or Socket Index=1, Second Bone or Socket Index=0)
   &FN_InitializeAnimTrailBoneIndecies(AnimTrail Selection=1, First Bone or Socket Index=3, Second Bone or Socket Index=2) {module=FN_InitializeAnimTrailBoneIndecies001}
   /SetVariables(Particles.AlphaFadeIn:NiagaraFloat=$System.SystemFadeIn) {isSetParameters=true}
   &FN_InitializeAnimTrail(
    AnimTrailWidth=0.02
    Color=&MakeLinearColor(
     R:NiagaraFloat=1
     G:NiagaraFloat=1
     B:NiagaraFloat=1
     A:NiagaraFloat=&Multiply_Float(A=$Particles.AlphaFadeIn, B=0.5)
    )
    Lifetime=&MakeCustomFloatFromBool(Bool=$Particles.Locked, True Float=100, False Float=1)
    Position Offset=&FN_FrontEndVector3Modifier(Gameplay=<0 1 0>, Front End=<0 20 0>)
   )
   &AddVelocity(Velocity=&Multiply_VectorByFloat(Vector=<0 400 0>, Float=&MakeCustomFloatFromBool(Bool=$User.bIsFrontEnd, True Float=-1, False Float=1)))
   &DynamicMaterialParameters(
    Index 0 Param 1=1
    Module.Param0WriteEnabled=true
    Module.Param1WriteEnabled=true
    Module.Param2WriteEnabled=true
    Module.Param3WriteEnabled=true
   )
  ]
  particleUpdateScript [
   &SolveForcesAndVelocity(Clamp Velocity=false, Limit Acceleration=false)
   &FN_UpdateAnimTrail(
    Locked Particle Color:LinearColor=&MultiplyLinearColor(
     Linear Color=$Particles.Initial.Color
     Scale Factor=&MakeFloat4FromVectorAndFloat("Vector (RGB)"=<1 1 1>, "Float (Alpha)"=$System.FN_Integrator001.Value)
    )
    Locked Particle RibbonWidth=1
    SkeletalMesh=$User.SkeletalMesh
   ) {module=FN_UpdateAnimTrail001}
   /SetVariables(Particles.RibbonU0Override=$Particles.NormalizedAge) {isSetParameters=true, module=SetVariables_#4}
   &Color(
    Color=<rgba 5 1.8325 0.2487 1>
    Scale Alpha=&Multiply_Float(
     A=&Multiply_Float(A=0.5, B=$Particles.AlphaFadeIn)
     B=&FN_FrontEndFloatModifier(Gameplay=$System.FN_Integrator001.Value, Front End=1)
    )
   ) {enabled=false}
   &ScaleColor(
    Scale Alpha=&Lerp_Float(A=0, B=1, Alpha=$System.Timeline.Time)
    Scale RGB=&Lerp_Vector(
     A=&MakeVectorFromLinearColorRGB(Color=<rgba 0.9 0.4781 0.0562 1>)
     B=&MakeVectorFromLinearColorRGB(Color=<rgba 0.6376 0.0398 0.0398 1>)
     Alpha=$System.Timeline.Time
    )
    ScaleA=true
    ScaleRGB=true
   )
  ]
  rendererClasses [&NiagaraRibbonRendererProperties]
  renderers [{rendererClass=&NiagaraRibbonRendererProperties}]
 }
 FN_AnimTrail002 {
  emitterSpawnScript [&FN_RibbonIDOffsetUtility()]
  emitterUpdateScript [
   &EmitterState()
   &FN_SpawnAnimTrail(Number of Anim Trails=2)
  ]
  particleSpawnScript [
   &FN_InitializeAnimTrailBoneIndecies(First Bone or Socket Index=1, Second Bone or Socket Index=0)
   &FN_InitializeAnimTrailBoneIndecies(AnimTrail Selection=1, First Bone or Socket Index=3, Second Bone or Socket Index=2) {module=FN_InitializeAnimTrailBoneIndecies001}
   /SetVariables(Particles.AlphaFadeIn:NiagaraFloat=$System.SystemFadeIn) {isSetParameters=true}
   &FN_InitializeAnimTrail(
    Color=&MakeLinearColor(
     R:NiagaraFloat=1
     G:NiagaraFloat=1
     B:NiagaraFloat=1
     A:NiagaraFloat=&Multiply_Float(A=$Particles.AlphaFadeIn, B=0.5)
    )
    Lifetime=&MakeCustomFloatFromBool(Bool=$Particles.Locked, True Float=100, False Float=1.5)
    Position Offset=&FN_FrontEndVector3Modifier(Gameplay=<0 -20 0>, Front End=<0 20 0>)
   )
   &AddVelocity(Velocity=&Multiply_VectorByFloat(Vector=<0 400 0>, Float=&MakeCustomFloatFromBool(Bool=$User.bIsFrontEnd, True Float=-1, False Float=1)))
  ]
  particleUpdateScript [
   &SolveForcesAndVelocity(Clamp Velocity=false, Limit Acceleration=false)
   &FN_UpdateAnimTrail(
    Locked Particle Color:LinearColor=&MultiplyLinearColor(
     Linear Color=$Particles.Initial.Color
     Scale Factor=&MakeFloat4FromVectorAndFloat("Vector (RGB)"=<1 1 1>, "Float (Alpha)"=$System.FN_Integrator001.Value)
    )
    Locked Particle RibbonWidth=1
    SkeletalMesh=$User.SkeletalMesh
   ) {module=FN_UpdateAnimTrail001}
   /SetVariables(Particles.RibbonU0Override=$Particles.NormalizedAge) {isSetParameters=true, module=SetVariables_#4}
   &Color(
    Color=&LerpLinearColors(StartColor=<rgba 1 0.0604 0.0604 1>, EndColor=<rgba 0.2 0.5 0.05 1>, LerpFactor=$System.Timeline002.Time)
    Scale Alpha=&Multiply_Float(
     A=&Multiply_Float(A=0.5, B=$Particles.AlphaFadeIn)
     B=&FN_FrontEndFloatModifier(Gameplay=$System.FN_Integrator001.Value, Front End=1)
    )
   )
  ]
  rendererClasses [&NiagaraRibbonRendererProperties]
  renderers [{rendererClass=&NiagaraRibbonRendererProperties}]
 }
 FN_AnimTrail003 {
  emitterSpawnScript [&FN_RibbonIDOffsetUtility()]
  emitterUpdateScript [
   &EmitterState()
   &FN_SpawnAnimTrail(AnimTrailWidthMode=::FromFirst, Number of Anim Trails=2)
  ]
  particleSpawnScript [
   &FN_InitializeAnimTrailBoneIndecies(First Bone or Socket Index=1, Second Bone or Socket Index=0)
   &FN_InitializeAnimTrailBoneIndecies(AnimTrail Selection=1, First Bone or Socket Index=3, Second Bone or Socket Index=2) {module=FN_InitializeAnimTrailBoneIndecies001}
   /SetVariables(Particles.AlphaFadeIn:NiagaraFloat=$System.SystemFadeIn) {isSetParameters=true}
   &FN_InitializeAnimTrail(
    AnimTrailWidth=0.1
    Color=&MakeLinearColor(
     R:NiagaraFloat=1
     G:NiagaraFloat=1
     B:NiagaraFloat=1
     A:NiagaraFloat=&Multiply_Float(A=$Particles.AlphaFadeIn, B=0.5)
    )
    Lifetime=&MakeCustomFloatFromBool(Bool=$Particles.Locked, True Float=100, False Float=1)
    Position Offset=&FN_FrontEndVector3Modifier(Gameplay=<0 1 0>, Front End=<0 20 0>)
   )
   &AddVelocity(Velocity=&Multiply_VectorByFloat(Vector=<0 400 0>, Float=&MakeCustomFloatFromBool(Bool=$User.bIsFrontEnd, True Float=-1, False Float=1)))
   &DynamicMaterialParameters(
    Index 0 Param 1=1
    Module.Param0WriteEnabled=true
    Module.Param1WriteEnabled=true
    Module.Param2WriteEnabled=true
    Module.Param3WriteEnabled=true
   )
  ]
  particleUpdateScript [
   &SolveForcesAndVelocity(Clamp Velocity=false, Limit Acceleration=false)
   &FN_UpdateAnimTrail(
    Locked Particle Color:LinearColor=&MultiplyLinearColor(
     Linear Color=$Particles.Initial.Color
     Scale Factor=&MakeFloat4FromVectorAndFloat("Vector (RGB)"=<1 1 1>, "Float (Alpha)"=$System.FN_Integrator001.Value)
    )
    Locked Particle RibbonWidth=1
    SkeletalMesh=$User.SkeletalMesh
   ) {module=FN_UpdateAnimTrail001}
   /SetVariables(Particles.RibbonU0Override=$Particles.NormalizedAge) {isSetParameters=true, module=SetVariables_#4}
   &Color(
    Color=<rgba 1.0406 5 0.2487 0.1>
    Scale Alpha=&Multiply_Float(
     A=&Multiply_Float(A=0.5, B=$Particles.AlphaFadeIn)
     B=&FN_FrontEndFloatModifier(Gameplay=$System.FN_Integrator001.Value, Front End=1)
    )
   )
  ]
  rendererClasses [&NiagaraRibbonRendererProperties]
  renderers [{rendererClass=&NiagaraRibbonRendererProperties}]
 }
 FN_AnimTrail004 {
  emitterSpawnScript [&FN_RibbonIDOffsetUtility()]
  emitterUpdateScript [
   &EmitterState()
   &FN_SpawnAnimTrail(Number of Anim Trails=2)
  ]
  particleSpawnScript [
   &FN_InitializeAnimTrailBoneIndecies(First Bone or Socket Index=1, Second Bone or Socket Index=0)
   &FN_InitializeAnimTrailBoneIndecies(AnimTrail Selection=1, First Bone or Socket Index=3, Second Bone or Socket Index=2) {module=FN_InitializeAnimTrailBoneIndecies001}
   /SetVariables(Particles.AlphaFadeIn:NiagaraFloat=$System.SystemFadeIn) {isSetParameters=true}
   &FN_InitializeAnimTrail(
    Color=&MakeLinearColor(
     R:NiagaraFloat=1
     G:NiagaraFloat=1
     B:NiagaraFloat=1
     A:NiagaraFloat=&Multiply_Float(A=$Particles.AlphaFadeIn, B=0.1)
    )
    Lifetime=&MakeCustomFloatFromBool(Bool=$Particles.Locked, True Float=100, False Float=1)
    Position Offset=&FN_FrontEndVector3Modifier(Gameplay=<0 -20 0>, Front End=<0 20 0>)
   )
   &AddVelocity(Velocity=&Multiply_VectorByFloat(Vector=<0 400 0>, Float=&MakeCustomFloatFromBool(Bool=$User.bIsFrontEnd, True Float=-1, False Float=1)))
  ]
  particleUpdateScript [
   &SolveForcesAndVelocity(Clamp Velocity=false, Limit Acceleration=false)
   &FN_UpdateAnimTrail(
    Locked Particle Color:LinearColor=&MultiplyLinearColor(
     Linear Color=$Particles.Initial.Color
     Scale Factor=&MakeFloat4FromVectorAndFloat("Vector (RGB)"=<1 1 1>, "Float (Alpha)"=$System.FN_Integrator001.Value)
    )
    Locked Particle RibbonWidth=1
    SkeletalMesh=$User.SkeletalMesh
   ) {module=FN_UpdateAnimTrail001}
   /SetVariables(Particles.RibbonU0Override=$Particles.NormalizedAge) {isSetParameters=true, module=SetVariables_#4}
   &Color(
    Color=<rgba 1.0406 5 0.2487 0.1>
    Scale Alpha=&Multiply_Float(
     A=&Multiply_Float(A=0.5, B=$Particles.AlphaFadeIn)
     B=&FN_FrontEndFloatModifier(Gameplay=$System.FN_Integrator001.Value, Front End=1)
    )
   ) {enabled=false}
   &ScaleColor(
    Scale Alpha=&Lerp_Float(A=0, B=0.4, Alpha=$System.Timeline.Time)
    Scale RGB=&MakeVectorFromLinearColorRGB(Color=&LerpLinearColors(StartColor=<rgba 0.0604 1 0.217 1>, EndColor=<rgba 0.7308 1 0.1925 1>, LerpFactor=$System.Timeline002.Time))
    ScaleA=true
    ScaleRGB=true
   )
  ]
  rendererClasses [&NiagaraRibbonRendererProperties]
  renderers [{rendererClass=&NiagaraRibbonRendererProperties}]
 }
 Sparkles_Beard {
  emitterUpdateScript [
   &EmitterState() {module=EmitterState001}
   &SpawnPerUnit(
    Module.Use Max Movement Threshold=true
    Module.Use Movement Tolerance=false
    Spawn Probability:NiagaraFloat=&FN_FrontEndFloatModifier(Gameplay=1, Front End=0)
    Spawn Spacing=50
    Use Spawn Probability=true
   )
   &SpawnRate(
    Spawn Probability=&FloatFromCurve(FloatCurve=DataInterface() {propertyValues {Curve {keys [<at 0 1>,<at 0.3 0>]}}}, CurveIndex=$Emitter.Age)
    SpawnRate=50
    Use Spawn Probability=true
   ) {enabled=false}
   &SpawnBurst_Instantaneous(Spawn Count=50, Use Spawn Probability=false) {enabled=false}
  ]
  particleSpawnScript [
   &InitializeParticle(
    Color=&MakeLinearColorFromVectorAndFloat(
     "Vector (RGB)"=&MakeVectorFromLinearColorRGB(Color=&HueShiftLinearColor(
       Hue=&Divide_Float(
        A=&SelectFloatFromArray(
         Array Sampling Mode=::NewEnumerator0
         Direct Array Index=?
         Float Selection Array=DataInterface() {propertyValues {FloatData [358,121,5,37]}}
         Random Seed=0
        )
        B=-360
       )
       Saturation=0.9
       Value=2
       Linear Color=$User.BaseHSVColor
      ))
     "Float (Alpha)"=1
    )
    Lifetime=&UniformRangedFloat(Minimum=0.3, Maximum=0.4)
    Mesh Scale Mode=::NewEnumerator1
    Mesh Uniform Scale Max=1.2
    Mesh Uniform Scale Min=0.8
    Sprite Rotation Mode=::NewEnumerator2
    Sprite Size Mode=::NewEnumerator3
    Uniform Sprite Size=&UniformRangedFloat(Minimum=5, Maximum=25)
    UsePositionOffset=false
    Write Material Random=true
   ) {module=InitializeParticle001}
   &AddVelocity(Velocity=<0 200 0>) {enabled=false}
   &InheritVelocity(
    Inherited Velocity Amount Scale=&VectorFromFloat(Value=0.2)
    Speed Limit=false
    UseSpeedThreshold=false
   ) {enabled=false}
   &ShapeLocation(
    MaskSpawnGroup=false
    Non Uniform Scale=<2 2 0.2>
    Offset=<0 80 75>
    Offset Coordinate Space=::Local
    Offset Mode=::NewEnumerator0
    Override Seed=false
    Sphere Radius=12
    Sphere Surface Distribution=1
    UseOwnerScale=true
   ) {enabled=false}
   &ShapeLocation(
    Cylinder Height=0
    Cylinder Height Midpoint=0
    Cylinder Radius=22
    Enable Lathe Profile=true
    Hemicircle Internal Cap=false
    Hemicircle Y=true
    MaskSpawnGroup=false
    Non Uniform Scale=<2 2 1>
    Offset=&Lerp_Vector(A=<10 85 88>, B=<7 80 78>, Alpha=$User.ForwardDot)
    Offset Coordinate Space=::Local
    Offset Mode=::NewEnumerator0
    Override Seed=false
    Rotation Angle Type:ENiagara_AngleInput=::NewEnumerator0
    Rotation Mode=::NewEnumerator2
    Shape Primitive=::NewEnumerator1
    Surface Only=true
    UseOwnerScale=true
    Yaw / Pitch / Roll=<0 0 18>
   ) {module=ShapeLocation001}
   &InitialMeshOrientation(Use Rotation Vector=false)
   &ApplyInitialForces()
  ]
  particleUpdateScript [
   &ParticleState()
   &ScaleMeshSize(Scale Factor=&VectorFromFloat(Value=&FloatFromCurve(FloatCurve=DataInterface() {propertyValues {Curve {keys [
          {interpMode=RCIM_Cubic, tangentMode=RCTM_User, time=0, value=1}
          {arriveTangent=-3.0133, interpMode=RCIM_Cubic, leaveTangent=-3.0133, tangentMode=RCTM_User, time=1, value=0}
         ]}}})))
   &GravityForce(Gravity=<0 0 200>)
   /Niagara/Modules/Update/Forces/CurlNoiseForce(Noise Frequency=0.02, Noise Quality / Cost=::NewEnumerator0, Noise Strength=500, Pan Noise=false)
   &Drag(Drag=5, Use Linear Drag=true, Use Rotational Drag=false)
   &InheritSourceMovement(
    Applied Force Scale=<-1 -1 -1>
    Applied Position Scale=&VectorFromFloat(Value=0.8)
    ApplyForce=false
    ApplyPosition=true
    ApplyPositionFalloff=false
   )
   &SolveForcesAndVelocity(Clamp Velocity=false, Limit Acceleration=false)
   &ScaleColor(
    Scale Alpha=&FloatFromCurve(FloatCurve=DataInterface() {propertyValues {Curve {keys [<at -0.0092 0.0144>,<at 0.0967 0.9767>,<at 1.0011 0.002>]}}})
    ScaleA=true
    ScaleRGB=true
   )
   /SetVariables(Particles.Color2=&MultiplyLinearColor(Linear Color=$Particles.Color, Scale Factor=<2 2 2 0.2>)) {isSetParameters=true, module=SetVariables_#5}
   &ScaleSpriteSize(Scale Factor=&Vector2DFromFloat(Value=&FloatFromCurve(FloatCurve=DataInterface() {propertyValues {Curve {keys [
          {interpMode=RCIM_Cubic, tangentMode=RCTM_User, time=0, value=0.4}
          {interpMode=RCIM_Cubic, tangentMode=RCTM_User, time=0.5, value=1}
          {tangentMode=RCTM_User, time=1, value=1}
         ]}}})))
   /SetVariables(Particles.SpriteSize2=&Vector2DFromFloat(Value=&FloatFromCurve(
      FloatCurve=DataInterface() {propertyValues {Curve {keys [
          {arriveTangent=-5.2489, interpMode=RCIM_Cubic, leaveTangent=-5.2489, tangentMode=RCTM_User, time=0, value=1}
          {arriveTangent=-1, interpMode=RCIM_Cubic, leaveTangent=-1, time=0.2135, value=0.4}
          {interpMode=RCIM_Cubic, tangentMode=RCTM_User, time=1, value=0}
         ]}}}
      Scale Curve=&Lerp_Float(A=&MakeFloatFromVector2d(Vector 2D=$Particles.Initial.SpriteSize), B=5, Alpha=0.7)
     ))) {isSetParameters=true, module=SetVariables_#6}
  ]
  rendererClasses [&NiagaraSpriteRendererProperties, &NiagaraMeshRendererProperties]
  renderers [
   {rendererClass=&NiagaraSpriteRendererProperties}
   {rendererClass=&NiagaraSpriteRendererProperties, rendererIndex=1}
   {rendererClass=&NiagaraMeshRendererProperties, rendererIndex=2}
  ]
 }
 Sparkles_BeardFE {
  emitterUpdateScript [
   &EmitterState(Life Cycle Mode=::NewEnumerator1, Loop Behavior=::NewEnumerator1, Loop Delay=1.5, Loop Duration Mode=::NewEnumerator1, UseLoopDelay=true) {module=EmitterState001}
   &SpawnPerUnit(Module.Use Max Movement Threshold=true, Module.Use Movement Tolerance=false, Spawn Spacing=50, Use Spawn Probability=false) {enabled=false}
   &SpawnRate(
    Spawn Probability=&FN_FrontEndFloatModifier(Gameplay=0, Front End=1)
    SpawnRate=50
    Use Spawn Probability=true
   )
   &SpawnBurst_Instantaneous(Spawn Count=50, Use Spawn Probability=false) {enabled=false}
  ]
  particleSpawnScript [
   &InitializeParticle(
    Color=&MakeLinearColorFromVectorAndFloat(
     "Vector (RGB)"=&MakeVectorFromLinearColorRGB(Color=&HueShiftLinearColor(
       Hue=&Divide_Float(
        A=&SelectFloatFromArray(
         Array Sampling Mode=::NewEnumerator0
         Direct Array Index=?
         Float Selection Array=DataInterface() {propertyValues {FloatData [358,121,5,37]}}
         Random Seed=0
        )
        B=-360
       )
       Saturation=0.9
       Value=2
       Linear Color=$User.BaseHSVColor
      ))
     "Float (Alpha)"=1
    )
    Lifetime=&UniformRangedFloat(Minimum=0.3, Maximum=0.5)
    Mesh Scale Mode=::NewEnumerator1
    Mesh Uniform Scale Max=1.2
    Mesh Uniform Scale Min=0.8
    Sprite Rotation Mode=::NewEnumerator2
    Sprite Size Mode=::NewEnumerator3
    Uniform Sprite Size=&UniformRangedFloat(Minimum=5, Maximum=25)
    UsePositionOffset=false
    Write Material Random=true
   ) {module=InitializeParticle001}
   &AddVelocity(Velocity=<0 -500 0>)
   &InheritVelocity(
    Inherited Velocity Amount Scale=&VectorFromFloat(Value=0.2)
    Speed Limit=false
    UseSpeedThreshold=false
   ) {enabled=false}
   &ShapeLocation(
    MaskSpawnGroup=false
    Non Uniform Scale=<2 2 0.2>
    Offset=<0 80 75>
    Offset Coordinate Space=::Local
    Offset Mode=::NewEnumerator0
    Override Seed=false
    Sphere Radius=12
    Sphere Surface Distribution=1
    UseOwnerScale=true
   ) {enabled=false}
   &ShapeLocation(
    Cylinder Height=0
    Cylinder Height Midpoint=0
    Cylinder Radius=22
    Enable Lathe Profile=true
    Hemicircle Internal Cap=false
    Hemicircle Y=true
    MaskSpawnGroup=false
    Non Uniform Scale=<2 2 1>
    Offset=<-10 -85 88>
    Offset Coordinate Space=::Local
    Offset Mode=::NewEnumerator0
    Override Seed=false
    Rotation Angle Type:ENiagara_AngleInput=::NewEnumerator0
    Rotation Mode=::NewEnumerator2
    Shape Primitive=::NewEnumerator1
    Surface Only=true
    UseOwnerScale=true
    Yaw / Pitch / Roll=<0 0 18>
   ) {module=ShapeLocation001}
   &InitialMeshOrientation(Use Rotation Vector=false)
   &ApplyInitialForces()
  ]
  particleUpdateScript [
   &ParticleState()
   &ScaleMeshSize(Scale Factor=&VectorFromFloat(Value=&FloatFromCurve(FloatCurve=DataInterface() {propertyValues {Curve {keys [
          {interpMode=RCIM_Cubic, tangentMode=RCTM_User, time=0, value=1}
          {arriveTangent=-3.0133, interpMode=RCIM_Cubic, leaveTangent=-3.0133, tangentMode=RCTM_User, time=1, value=0}
         ]}}})))
   &GravityForce(Gravity=<0 0 -50>) {enabled=false}
   &CurlNoiseForce(Noise Frequency=0.02, Noise Quality / Cost=::NewEnumerator0, Noise Strength=500, Pan Noise=false) {enabled=false}
   &Drag(Drag=2, Use Linear Drag=true, Use Rotational Drag=false) {enabled=false}
   &InheritSourceMovement(
    Applied Force Scale=<-1 -1 -1>
    Applied Position Scale=&VectorFromFloat(Value=0.8)
    ApplyForce=false
    ApplyPosition=true
    ApplyPositionFalloff=false
   ) {enabled=false}
   &SolveForcesAndVelocity(Clamp Velocity=false, Limit Acceleration=false)
   &ScaleColor(
    Scale Alpha=&FloatFromCurve(FloatCurve=DataInterface() {propertyValues {Curve {keys [<at -0.0092 0.0144>,<at 0.0967 0.9767>,<at 1.0011 0.002>]}}})
    ScaleA=true
    ScaleRGB=true
   )
   /SetVariables(Particles.Color2=&MultiplyLinearColor(Linear Color=$Particles.Color, Scale Factor=<2 2 2 0.2>)) {isSetParameters=true, module=SetVariables_#5}
   &ScaleSpriteSize(Scale Factor=&Vector2DFromFloat(Value=&FloatFromCurve(FloatCurve=DataInterface() {propertyValues {Curve {keys [
          {interpMode=RCIM_Cubic, tangentMode=RCTM_User, time=0, value=0.4}
          {interpMode=RCIM_Cubic, tangentMode=RCTM_User, time=0.5, value=1}
          {tangentMode=RCTM_User, time=1, value=1}
         ]}}})))
   /SetVariables(Particles.SpriteSize2=&Vector2DFromFloat(Value=&FloatFromCurve(
      FloatCurve=DataInterface() {propertyValues {Curve {keys [
          {arriveTangent=-5.2489, interpMode=RCIM_Cubic, leaveTangent=-5.2489, tangentMode=RCTM_User, time=0, value=1}
          {arriveTangent=-1, interpMode=RCIM_Cubic, leaveTangent=-1, time=0.2135, value=0.4}
          {interpMode=RCIM_Cubic, tangentMode=RCTM_User, time=1, value=0}
         ]}}}
      Scale Curve=&Lerp_Float(A=&MakeFloatFromVector2d(Vector 2D=$Particles.Initial.SpriteSize), B=5, Alpha=0.7)
     ))) {isSetParameters=true, module=SetVariables_#6}
  ]
  rendererClasses [&NiagaraSpriteRendererProperties, &NiagaraMeshRendererProperties]
  renderers [
   {rendererClass=&NiagaraSpriteRendererProperties}
   {rendererClass=&NiagaraSpriteRendererProperties, rendererIndex=1}
   {rendererClass=&NiagaraMeshRendererProperties, rendererIndex=2}
  ]
 }
 SpeedLine {
  emitterUpdateScript [
   &EmitterState()
   &SpawnRate(
    SpawnRate=&MakeCustomFloatFromBool(Bool=$User.bIsFrontEndPreview, True Float=50, False Float=90)
    Use Spawn Probability=true
   ) {module=SpawnRate001}
  ]
  particleSpawnScript [
   &InitializeParticle(
    Color=&MakeLinearColorFromVectorAndFloat(
     "Vector (RGB)"=<1 1 1>
     "Float (Alpha)"=&Multiply_Float(
      A=&UniformRangedFloat(Minimum=0.1, Maximum=0.18)
      B=&MakeCustomFloatFromBool(Bool=$User.bIsFrontEndPreview, True Float=0.5, False Float=1)
     )
    )
    Lifetime Max=0.16
    Lifetime Min=0.1
    Lifetime Mode=::NewEnumerator1
    Sprite Size Max=<3 10>
    Sprite Size Min=<2 2>
    Sprite Size Mode=::NewEnumerator2
    UsePositionOffset=false
    Write Material Random=true
   )
   &TorusLocation(
    Align Torus With Axis=true
    Handle Radius=0
    Large Radius:NiagaraFloat=&RandomRangeFloat(Minimum=80, Maximum=150)
    MaskSpawnGroup=false
    Module.Override Randomness=false
    Offset=<0 -180 0>
    Torus Axis=<0 1 0>
    Torus Origin:NiagaraPosition=&Add_Vector(
     A=Linked() {linkedVariable {name=Particles.Position, type=NiagaraPosition}}
     B=&MakeVector(X=0, Y=&MakeCustomFloatFromBool(Bool=$User.bIsFrontEndPreview, True Float=200, False Float=0), Z=0)
    )
   )
   &AddVelocityFromPoint(
    Module.Use Falloff Distance=false
    Origin Offset=<0 -500 0>
    Velocity Strength=&Multiply_Float(A=600, B=&MakeCustomFloatFromBool(Bool=$User.bIsFrontEndPreview, True Float=-1, False Float=1))
   )
   &AddVelocity(Velocity=&Multiply_Vector(
     A=&RandomRangeVector(Minimum=<0 300 0>, Maximum=<0 700 0>)
     B=&MakeVector(X=1, Y=&MakeCustomFloatFromBool(Bool=$User.bIsFrontEndPreview, True Float=-1, False Float=1), Z=1)
    ))
  ]
  particleUpdateScript [
   &SolveForcesAndVelocity(Clamp Velocity=false, Limit Acceleration=false)
   &ParticleState()
   &ScaleSpriteSize(Scale Factor=&Vector2DFromCurve(
     Vector2Curve=DataInterface() {propertyValues {XCurve {keys [<at 0 0.5 RCIM_Cubic>,<at 0.5 1 RCIM_Cubic>]}, YCurve {keys [<at 0 0 RCIM_Cubic>,<at 0.5 1 RCIM_Cubic>]}}}
     ScaleCurve=&Vector2DFromFloat(Value=1)
    ))
   &ScaleSpriteSizeBySpeed(Max Scale Factor=<1 50>, Min Scale Factor=<0.01 0.05>, Velocity Threshold=1)
   &ScaleColor(
    Scale Alpha=&FloatFromCurve(
     FloatCurve=DataInterface() {propertyValues {Curve {keys [
         <at 0 0 RCIM_Cubic>
         <at 0.3759 1.0207>
         {arriveTangent=-2.221, interpMode=RCIM_Cubic, leaveTangent=-2.221, tangentMode=RCTM_User, time=1, value=0}
        ]}}}
     Scale Curve=&MakeCustomFloatFromBool(
      Bool=$User.bIsFrontEndPreview
      True Float=&MakeCustomFloatFromBool(Bool=$User.bIsFullyDeployed, True Float=1, False Float=0)
      False Float=&Lerp_Float(A=0, B=1, Alpha=$User.ForwardDot)
     )
    )
    Scale RGB=&VectorFromFloat(Value=1)
    ScaleA=true
    ScaleRGB=true
   )
  ]
  rendererClasses [&NiagaraSpriteRendererProperties]
  renderers [{rendererClass=&NiagaraSpriteRendererProperties}]
 }
}
summary {
 systemName=NS_Glider_Default_Jolly_Trails
 emitters [
  {emitterName=FN_AnimTrail002, rendererClasses [&NiagaraRibbonRendererProperties]}
  {emitterName=SpeedLine, rendererClasses [&NiagaraSpriteRendererProperties]}
  {emitterName=FN_AnimTrail, rendererClasses [&NiagaraRibbonRendererProperties]}
  {emitterName=FN_AnimTrail001, rendererClasses [&NiagaraRibbonRendererProperties]}
  {emitterName=FN_AnimTrail003, rendererClasses [&NiagaraRibbonRendererProperties]}
  {emitterName=FN_AnimTrail004, rendererClasses [&NiagaraRibbonRendererProperties]}
  {emitterName=Sparkles_Beard, rendererClasses [&NiagaraSpriteRendererProperties, &NiagaraMeshRendererProperties]}
  {
   emitterName=Sparkles_BeardFE
   rendererClasses [&NiagaraSpriteRendererProperties, &NiagaraMeshRendererProperties]
  }
 ]
 userVariables [
  User.BaseHSVColor {type=LinearColor, description="", defaultValue {struct=LinearColor, value=<rgba 1 0 0 1>}}
  User.bIsFrontEnd {type=NiagaraBool}
  User.bIsFrontEndPreview {type=NiagaraBool}
  User.bIsFullyDeployed {type=NiagaraBool}
  User.DelayDeployTime {type=NiagaraFloat, defaultValue=0, description=""}
  User.ForwardDot {type=NiagaraFloat, defaultValue=0, description=""}
  User.GlowHue {type=NiagaraFloat, defaultValue=1, description=""}
  User.RibbonIDOffset {type=NiagaraInt32}
  User.RightDot {type=NiagaraFloat, defaultValue=0, description=""}
  User.SkeletalMesh {
   type=NiagaraDataInterfaceSkeletalMesh
   defaultValue {
    struct=NiagaraExt_VariableValue_DataInterface
    value {dataInterface=&self:NiagaraDataInterfaceSkeletalMesh_0}
   }
  }
 ]
}
